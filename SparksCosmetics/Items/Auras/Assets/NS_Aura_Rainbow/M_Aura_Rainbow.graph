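# Translucent, unlit, two-sided material for the Sparks rainbow aura, used on Niagara ribbons, mesh particles and beam trails.
# Panning swirl noise distorts UVs by audio-reactive Amplitude, shaping a generated band and offsetting its rainbow curve-atlas lookup.
# Emissive combines color, band, time-of-day brightness and particle color, then remaps to HDR.
# Opacity combines band, edge masks, depth fade and particle alpha; ES2/ES3_1 use only particle alpha.
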Material /SparksCosmetics/Items/Auras/Assets/NS_Aura_Rainbow/M_Aura_Rainbow
graph {
 properties {
  bUsedWithBeamTrails=true
  bUsedWithNiagaraMeshParticles=true
  bUsedWithNiagaraRibbons=true
  blendMode=BLEND_Translucent
  pixelDepthOffsetMode=PDOM_Legacy
  shadingModel=MSM_Unlit
  translucencyPass=MTP_BeforeDOF
  twoSided=true
  usedShadingModels=Unlit
 }
 nodes [
  Constant() {r=50}
  n1 = ParticleColor()
  n2 = Multiply(
   A=TextureCoordinate()
   B=AppendVector(
    A=ScalarParameter() {defaultValue=0.1, group=Texture, parameterName=Noise X Scale, sortPriority=3}
    B=ScalarParameter() {defaultValue=3, group=Texture, parameterName=Noise Y Scale, sortPriority=4}
   )
  )
  n3 = Constant2Vector() {g=0.5, r=0.5}
  n4 = Constant() {r=0.25}
  n5 = Constant()
  n6 = Constant() {r=0.4}
  n7 = TextureSampleParameter2D(UVs=Panner(
    Coordinate=MaterialFunctionCall(UVs=@n2, Rotation Center=@n3, "Rotation Angle (0-1)"=@n4) {materialFunction=&CustomRotator}.Rotated Values
    Speed=AppendVector(A=@n5, B=@n6)
   ) {speedY=1}) {
   desc=Noise/Distortion
   group=Texture
   parameterName=Texture
   samplerType=SAMPLERTYPE_LinearGrayscale
   sortPriority=0
   texture=/Game/Global/Textures/Default/Noise/T_SwirlNoise_N_height
  }
  n8 = Power(Base=@n7.R, Exp=ScalarParameter() {defaultValue=5, parameterName=Param_2}) {constExponent=1.5}
  n9 = ScalarParameter() {defaultValue=3, parameterName=Param_3}
  n10 = LinearInterpolate(
   A=TextureCoordinate()
   B=OneMinus(None=Multiply(A=@n8, B=@n9) {constB=2.5})
   Alpha=Multiply(A=CollectionParameter() {collection=/FMAudioReactiveData/AudioAnalysis/MPC_Sparks_Audio_Parameters, parameterName=Amplitude}) {constB=0.1}
  ) {constAlpha=0.2}
  n11 = MaterialFunctionCall!(Width=Constant() {r=0.3}, Sharpness, Offset=Constant(), Direction Switch=StaticBool() {value=true}, Compare, Input Coordinates=@n10) {materialFunction=/Engine/Functions/Engine_MaterialFunctions03/Procedurals/GeneratedBand}
  n12 = Power(
   Base=ComponentMask(None=TextureCoordinate()) {r=true}
   Exp=ScalarParameter() {defaultValue=1.1, group=Opacity, parameterName=Left-Right Contrast, sortPriority=2}
  )
  n13 = TextureCoordinate()
  n14 = Power(
   Base=OneMinus(None=ComponentMask(None=@n13) {r=true})
   Exp=ScalarParameter() {defaultValue=2, group=Opacity, parameterName=Right-Left Contrast, sortPriority=3}
  )
  n15 = TextureCoordinate()
  n16 = Add(A=Power(
    Base=ComponentMask(None=@n15) {g=true}
    Exp=ScalarParameter() {defaultValue=1.9, group=Opacity, parameterName=Top-Bottom Contrast, sortPriority=4}
   )) {constB=-0.01}
  n17 = ComponentMask(None=TextureCoordinate()) {g=true}
  n18 = Add(A=Power(Base=OneMinus(None=@n17), Exp=ScalarParameter() {defaultValue=1.9, group=Opacity, parameterName=Bottom-Top Contrast, sortPriority=5})) {constB=-0.01}
  n19 = Multiply(
   A=SphereMask(A=TextureCoordinate(), B=Constant2Vector() {g=0.5, r=0.5}) {attenuationRadius=0.5, hardnessPercent=0}
   B=Multiply(A=Multiply(A=@n12, B=@n14), B=Multiply(A=@n16, B=@n18))
  )
  n20 = ScalarParameter() {defaultValue=500, group=Opacity, parameterName=Opacity_Intensity_Compensator, sortPriority=1}
  n21 = Clamp(None=Multiply(A=@n11.Result, B=Multiply(A=@n19, B=@n20)))
  n22 = Multiply(A=Multiply(A=@n1.A, B=DepthFade(Opacity=@n21) {fadeDistanceDefault=5}))
  n23 = Multiply(A=@n1.A)
  FeatureLevelSwitch(Default=@n22, ES2=@n23, ES3_1=@n23, SM4_REMOVED=@n22, SM5=@n22, SM6=@n22)
  n24 = ScalarParameter() {defaultValue=0.1, parameterName=ColorScroll}
  n25 = TextureCoordinate()
  n26 = AppendVector(
   A=ScalarParameter() {defaultValue=1, group=Texture, parameterName=Texture X Scale, sortPriority=3}
   B=ScalarParameter() {desc=does nothing, group=Texture, parameterName=Texture Y Scale, sortPriority=4}
  )
  n27 = Multiply(A=Add(A=Add(A=@n24, B=@n25), B=Multiply(A=@n25, B=@n26)), B=@n10) {desc=Makes noise affect color}
  n28 = CurveAtlasRowParameter(CurveTime=TextureSample(UVs=Clamp(None=@n27) {desc=Prevents wrap}) {samplerType=SAMPLERTYPE_LinearColor, texture=/Game/Global/Textures/Default/Masks_and_Gradients/linear_gradient}.R) {
   atlas=&/CA_Aura_Rainbow
   curve=&/CC_Rainbow
  }
  n29 = ScalarParameter() {defaultValue=0.55, parameterName=Brightness_Morning}
  n30 = ScalarParameter() {defaultValue=0.55, parameterName=Brightness_Day}
  n31 = ScalarParameter() {defaultValue=1, parameterName=Brightness_Evening}
  n32 = ScalarParameter() {defaultValue=0.65, parameterName=Brightness_Night}
  n33 = Multiply(
   A=Multiply(
    A=Multiply(A=@n28, B=@n11.Result)
    B=MaterialFunctionCall(Color_Morning=@n29, Color_Day=@n30, Color_Evening=@n31, Color_Night=@n32) {materialFunction=&Color_Time_of_Day}.Result
   )
   B=@n1.RGB
  )
  MaterialFunctionCall(Emissive=MaterialFunctionCall(LDR Input=Clamp(None=@n33), HDR Tint=@n33) {materialFunction=&DeriveHDRfromLDR}.RemappedRange) {materialFunction=&MF_FX_EmissiveAdjustments}
 ]
}
