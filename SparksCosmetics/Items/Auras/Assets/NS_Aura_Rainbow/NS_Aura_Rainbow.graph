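# Niagara system for a rainbow aura cosmetic that attaches glows, rays and sparkles to a character's skeletal mesh.
# Glow002 follows seven sampled bones with hue-cycling sprites, while burst and persistent ray emitters rotate layered rainbow sprites at another bone.
# User flags adapt size for large bodies and brightness for the front end, and time-of-day scalars tune the rays.
# Several emitters, including aurora meshes and debug, are disabled.
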NiagaraSystem /SparksCosmetics/Items/Auras/Assets/NS_Aura_Rainbow/NS_Aura_Rainbow
dependencies {
 usedDataInterfaces [
  &NiagaraDataInterfaceArrayFloat
  &NiagaraDataInterfaceArrayColor
  &NiagaraDataInterfaceCurve
  &NiagaraDataInterfaceArrayFloat3
  &NiagaraDataInterfaceColorCurve
  &NiagaraDataInterfaceArrayInt32
  &NiagaraDataInterfaceVector2DCurve
  &NiagaraDataInterfaceVectorCurve
  &NiagaraDataInterfaceSkeletalMesh
 ]
 usedDynamicInputs [
  &FNTimeOfDayColor
  &MakeVector
  &SimulationPosition
 ]
 usedModules [
  &self:SystemSpawnScript.SystemScriptSource.SystemScriptGraph.NiagaraNodeAssignment_0.SetVariables_#1
  &SystemState
 ]
}
emitters {
 Aurora_Mesh001 {
  bEnabled=false
  emitterUpdateScript [
   &EmitterState(
    Inactive Response=::NewEnumerator1
    Life Cycle Mode=::NewEnumerator1
    Loop Behavior=::NewEnumerator1
    Loop Delay=0.4
    Loop Duration Mode=::NewEnumerator1
    UseLoopDelay=true
   )
   &SpawnBurst_Instantaneous(Use Spawn Probability=false, UseLoopCountLimit=false)
  ]
  particleSpawnScript [
   &InitializeParticle(
    Color=<rgba 1 1 1 1.5>
    Mesh Scale=&Multiply_VectorByFloat(
     Vector=<0.2 0.1 0.13>
     Float=&MakeCustomFloatFromBool(Bool=$User.IsLargeBody, True Float=1.8, False Float=1.6)
    )
    Mesh Scale Mode=::NewEnumerator4
    Position Offset=&SelectVectorFromArray(
     Direct Array Index=$Particles.PartitionParticles.Partition
     Vector Selection Array=DataInterface() {propertyValues {FloatData [<0 0 75>, <1 1 65>]}}
    )
    UsePositionOffset=false
    Write Material Random=true
   )
   &InitialMeshOrientation(Mesh Orientation Mode=::NewEnumerator2, Use Rotation Vector=true)
  ]
  particleUpdateScript [
   &ParticleState(
    Kill Particles When Lifetime Has Elapsed=false
    Let Infinitely Lived Particles Die When Emitter Deactivates=true
    Loop Particles Lifetime=true
   )
   &SkeletalMeshLocation(
    Bone Sampling Mode=::NewEnumerator4
    MaskSpawnGroup=false
    Orientation Sampling=::NewEnumerator1
    Sampled Position Offset=<0 0 37>
    Sampled Position Offset Coordinate Space:ENiagaraSkelMeshTransforms=::NewEnumerator1
    Skeletal Mesh=$System.Skeletal Mesh
   )
   &CameraOffset(Camera Offset Amount=-30)
  ]
  rendererClasses [&NiagaraMeshRendererProperties]
  renderers [{rendererClass=&NiagaraMeshRendererProperties, rendererIndex=0}]
 }
 Aurora_Mesh_Scaling {
  bEnabled=false
  emitterUpdateScript [
   &EmitterState(
    Inactive Response=::NewEnumerator1
    Life Cycle Mode=::NewEnumerator1
    Loop Behavior=::NewEnumerator1
    Loop Duration Mode=::NewEnumerator1
    UseLoopDelay=false
   )
   &SpawnBurst_Instantaneous(Use Spawn Probability=false, UseLoopCountLimit=false)
  ]
  particleSpawnScript [
   &InitializeParticle(
    Color=<rgba 1 1 1 1.5>
    Lifetime=0.5
    Mesh Scale=&Multiply_VectorByFloat(
     Vector=<0.2 0.1 0.13>
     Float=&MakeCustomFloatFromBool(Bool=$User.IsLargeBody, True Float=1.8, False Float=1.6)
    )
    Mesh Scale Mode=::NewEnumerator4
    Position Offset=&SelectVectorFromArray(
     Direct Array Index=$Particles.PartitionParticles.Partition
     Vector Selection Array=DataInterface() {propertyValues {FloatData [<0 0 75>, <1 1 65>]}}
    )
    UsePositionOffset=false
    Write Material Random=true
   )
   &InitialMeshOrientation(Mesh Orientation Mode=::NewEnumerator2, Use Rotation Vector=true)
  ]
  particleUpdateScript [
   &ParticleState()
   &ScaleMeshSize(Scale Factor=&VectorFromCurve(
     VectorCurve=DataInterface() {propertyValues {
       XCurve {keys [{arriveTangent=0.0622, interpMode=RCIM_Cubic, leaveTangent=0.0622, tangentMode=RCTM_User, time=0, value=0.2}, <at 0.9 1.5>, <at 1 1.2>]}
       YCurve {keys [{arriveTangent=0.0622, interpMode=RCIM_Cubic, leaveTangent=0.0622, tangentMode=RCTM_User, time=0, value=0.2}, <at 0.9 1.5>, <at 1 1.2>]}
       ZCurve {keys [{arriveTangent=0.0622, interpMode=RCIM_Cubic, leaveTangent=0.0622, tangentMode=RCTM_User, time=0, value=0.2}, <at 0.9 1.5>, <at 1 1.2>]}
      }}
     Scale Curve=&Multiply_VectorByFloat(Vector=<1 1 1>, Float=0.7)
    ))
   &SkeletalMeshLocation(
    Bone Sampling Mode=::NewEnumerator4
    MaskSpawnGroup=false
    Orientation Sampling=::NewEnumerator1
    Sampled Position Offset=<0 0 37>
    Sampled Position Offset Coordinate Space:ENiagaraSkelMeshTransforms=::NewEnumerator1
    Skeletal Mesh=$System.Skeletal Mesh
   )
   &CameraOffset(Camera Offset Amount=-30)
   &ScaleColor(
    Scale Alpha=&FloatFromCurve(FloatCurve=DataInterface() {propertyValues {Curve {keys [
         {tangentMode=RCTM_User, time=0, value=0}
         {arriveTangent=-0.0083, interpMode=RCIM_Cubic, leaveTangent=-0.0083, tangentMode=RCTM_User, time=0.3, value=0}
         {arriveTangent=-0.4982, interpMode=RCIM_Cubic, leaveTangent=-0.4982, tangentMode=RCTM_User, time=0.7, value=1}
         {arriveTangent=-5.6148, interpMode=RCIM_Cubic, leaveTangent=-5.6148, tangentMode=RCTM_User, time=1, value=0}
        ]}}})
    ScaleA=true
    ScaleRGB=false
   )
  ]
  rendererClasses [&NiagaraMeshRendererProperties]
  renderers [{rendererClass=&NiagaraMeshRendererProperties, rendererIndex=0}]
 }
 FN_Debug {
  bEnabled=false
  emitterUpdateScript [
   &EmitterState()
   &SpawnBurst_Instantaneous(Spawn Count=1, Use Spawn Probability=false)
  ]
  particleSpawnScript [&InitializeParticle(
    Lifetime=5
    Mass Mode=::NewEnumerator1
    Position Offset=<40 0 160>
    Sprite Size=<75 75>
    Sprite Size Mode=::NewEnumerator4
    UsePositionOffset=true
    Write Material Random=true
   ) {module=InitializeParticle001}]
  particleUpdateScript [
   &ParticleState()
   /SetVariables(Particles.Debug=0.1) {isSetParameters=true}
   &DynamicMaterialParameters(
    Index 0 Param 1=&MakeCustomFloatFromBool(Bool=$User.IsFrontEnd, True Float=1, False Float=0)
    Module.Index0Param0Write=true
    Module.Index0Param1Write=true
    Module.Index0Param2Write=true
    Module.Index0Param3Write=true
   )
  ]
  rendererClasses [&NiagaraSpriteRendererProperties]
  renderers [{rendererClass=&NiagaraSpriteRendererProperties}]
 }
 Glow001 {
  bEnabled=false
  emitterUpdateScript [
   &EmitterState(
    Inactive Response=::NewEnumerator1
    Life Cycle Mode=::NewEnumerator1
    Loop Behavior=::NewEnumerator1
    Loop Duration Mode=::NewEnumerator1
    UseLoopDelay=false
   )
   &SpawnBurst_Instantaneous(Spawn Count=7, Use Spawn Probability=false, UseLoopCountLimit=false)
  ]
  particleSpawnScript [
   &PartitionParticles(Number Of Partitions=7)
   &InitializeParticle(
    Color Mode=::NewEnumerator0
    Lifetime=&RandomRangeFloat(Minimum=2, Maximum=3)
    Sprite Rotation Mode=::NewEnumerator0
    Sprite Size Mode=::NewEnumerator3
    Uniform Sprite Size=&Multiply_Float(
     A=&SelectFloatFromArray(
      Direct Array Index=&SpawnGroup()
      Float Selection Array=DataInterface() {propertyValues {FloatData [40,40,20,20,30,30]}}
     )
     B=&MakeCustomFloatFromBool(Bool=$User.IsLargeBody, True Float=1.5, False Float=1)
    )
    UsePositionOffset=false
    Write Material Random=true
   ) {module=InitializeParticle001}
   /SetVariables(Particles.glowScale:Vector2f=&Vector2DFromFloat(Value=&Multiply_Float(
      A=&SelectFloatFromArray(
       Direct Array Index=$Particles.PartitionParticles.Partition
       Float Selection Array=DataInterface() {propertyValues {FloatData [150,100,100,90,90,50,50]}}
      )
      B=&MakeCustomFloatFromBool(Bool=$User.IsLargeBody, True Float=1.2, False Float=1)
     ))) {isSetParameters=true, module=SetVariables_#2}
   /SetVariables(Particles.glowCameraOffset=-10) {isSetParameters=true, module=SetVariables_#3}
  ]
  particleUpdateScript [
   &ParticleState(
    Kill Particles When Lifetime Has Elapsed=false
    Let Infinitely Lived Particles Die When Emitter Deactivates=true
    Loop Particles Lifetime=true
   )
   &CameraOffset(Camera Offset Amount=5)
   /SetVariables(Particles.decalColor:LinearColor=&ColorFromCurve(ColorCurve=DataInterface() {propertyValues {
       AlphaCurve {keys [<at 0 2 RCIM_Cubic>,<at 1 2 RCIM_Cubic>]}
       BlueCurve {keys [
         <at 0.0411 1>
         <at 0.3308 1>
         <at 0.5699 0>
         <at 0.7921 0>
         {arriveTangent=4.8092, interpMode=RCIM_Cubic, leaveTangent=4.8092, time=0.8998, value=0.7382}
         <at 1 1>
        ]}
       GreenCurve {keys [
         <at 0.0411 0.3005>
         <at 0.3308 0>
         <at 0.5699 0.0118>
         <at 0.7921 1>
         {arriveTangent=-3.3638, interpMode=RCIM_Cubic, leaveTangent=-3.3638, time=0.8998, value=1}
         <at 1 0.3005>
        ]}
       RedCurve {keys [
         <at 0.0411 0>
         <at 0.3308 0.5569>
         <at 0.5699 1>
         <at 0.7921 0.3333>
         {arriveTangent=-1.6031, interpMode=RCIM_Cubic, leaveTangent=-1.6031, time=0.8998, value=0}
         <at 1 0>
        ]}
      }})) {isSetParameters=true, module=SetVariables_#4}
   &ScaleSpriteSize(
    Scale Sprite Size Mode=::NewEnumerator0
    Uniform Scale Factor=&Waveform(
     Global Phase=&Add_Float(A=$Emitter.Age, B=&RandomRangeFloat(Minimum=0, Maximum=5))
     "Waveform [1]"=::NewEnumerator0
     "[1] Frequency"=1
     "[1] Amplitude Min/Max"=<0.85 1>
    )
   )
   &SpriteRotationRate(Rotation Rate=&RandomRangeFloat(Minimum=10, Maximum=20))
   &SkeletalMeshLocation(
    Bone / Socket Index=&SelectIntFromArray(Direct Array Index=$Particles.PartitionParticles.Partition, Int Selection Array=DataInterface() {propertyValues {IntData [0,1,2,3,4,5,6]}})
    Bone Sampling Mode=::NewEnumerator4
    MaskSpawnGroup=false
    Orientation Sampling=::NewEnumerator1
    Sampled Position Offset Coordinate Space:ENiagaraSkelMeshTransforms=::NewEnumerator1
    Skeletal Mesh=$System.Skeletal Mesh
   ) {module=SkeletalMeshLocation003}
   &ScaleColor(
    Scale Alpha=0.4
    Scale RGB=&MakeVectorFromLinearColorRGB(Color=&ColorFromCurve(ColorCurve=DataInterface() {propertyValues {
        AlphaCurve {keys [<at 0 1 RCIM_Cubic>,<at 1 1 RCIM_Cubic>]}
        BlueCurve {keys [
          <at 0.0411 1>
          <at 0.3308 1>
          <at 0.5699 0>
          <at 0.7921 0>
          {arriveTangent=4.8091, interpMode=RCIM_Cubic, leaveTangent=4.8091, time=0.8998, value=0.7382}
          <at 1 1>
         ]}
        GreenCurve {keys [
          <at 0.0411 0.3005>
          <at 0.3308 0>
          <at 0.5699 0.0116>
          <at 0.7921 1>
          {arriveTangent=-3.3638, interpMode=RCIM_Cubic, leaveTangent=-3.3638, time=0.8998, value=1}
          <at 1 0.3005>
         ]}
        RedCurve {keys [
          <at 0.0411 0>
          <at 0.3308 0.5569>
          <at 0.5699 1>
          <at 0.7921 0.3333>
          {arriveTangent=-1.603, interpMode=RCIM_Cubic, leaveTangent=-1.603, time=0.8998, value=0}
          <at 1 0>
         ]}
       }}))
    ScaleA=true
    ScaleRGB=true
   )
   &DynamicMaterialParameters(
    Index 0 Param 1=&SelectFloatFromArray(Direct Array Index=$Particles.PartitionParticles.Partition, Float Selection Array=DataInterface() {propertyValues {FloatData [5,5,5,5,3,3]}})
    Index 0 Param 3=1
    Module.Param0WriteEnabled=true
    Module.Param1WriteEnabled=true
    Module.Param2WriteEnabled=true
    Module.Param3WriteEnabled=true
   )
  ]
  rendererClasses [&NiagaraSpriteRendererProperties]
  renderers [
   {rendererClass=&NiagaraSpriteRendererProperties}
   {rendererClass=&NiagaraSpriteRendererProperties, rendererIndex=1}
  ]
 }
 Glow002 {
  emitterUpdateScript [
   &EmitterState(
    Inactive Response=::NewEnumerator1
    Life Cycle Mode=::NewEnumerator1
    Loop Behavior=::NewEnumerator1
    Loop Duration Mode=::NewEnumerator1
    UseLoopDelay=false
   )
   &SpawnBurst_Instantaneous(Spawn Count=7, Use Spawn Probability=false, UseLoopCountLimit=false)
  ]
  particleSpawnScript [
   &PartitionParticles(Number Of Partitions=7)
   &InitializeParticle(
    Color Mode=::NewEnumerator0
    Lifetime=&RandomRangeFloat(Minimum=3, Maximum=4)
    Sprite Rotation Mode=::NewEnumerator0
    Sprite Size Mode=::NewEnumerator3
    Uniform Sprite Size=&Multiply_Float(
     A=&SelectFloatFromArray(
      Direct Array Index=$Particles.PartitionParticles.Partition
      Float Selection Array=DataInterface() {propertyValues {FloatData [50,40,40,40,40,40]}}
     )
     B=&MakeCustomFloatFromBool(Bool=$User.IsLargeBody, True Float=1.5, False Float=1)
    )
    UsePositionOffset=false
    Write Material Random=true
   ) {module=InitializeParticle001}
   /SetVariables(Particles.glowScale:Vector2f=&Vector2DFromFloat(Value=&Multiply_Float(
      A=&SelectFloatFromArray(
       Direct Array Index=$Particles.PartitionParticles.Partition
       Float Selection Array=DataInterface() {propertyValues {FloatData [100,90,90,90,90,50,50]}}
      )
      B=&MakeCustomFloatFromBool(Bool=$User.IsLargeBody, True Float=1.2, False Float=1)
     ))) {isSetParameters=true, module=SetVariables_#2}
   /SetVariables(Particles.glowCameraOffset=-10) {isSetParameters=true, module=SetVariables_#3}
  ]
  particleUpdateScript [
   &ParticleState(
    Kill Particles When Lifetime Has Elapsed=false
    Let Infinitely Lived Particles Die When Emitter Deactivates=true
    Loop Particles Lifetime=true
   )
   &CameraOffset(Camera Offset Amount=5)
   /SetVariables(Particles.decalColor:LinearColor=&ColorFromCurve(ColorCurve=DataInterface() {propertyValues {
       AlphaCurve {keys [<at 0 5 RCIM_Cubic>,<at 1 2 RCIM_Cubic>]}
       BlueCurve {keys [
         <at 0 1>
         <at 0.125 1>
         <at 0.375 0>
         <at 0.625 0>
         {arriveTangent=1.6, interpMode=RCIM_Cubic, leaveTangent=1.6, time=0.875, value=0.7382}
         <at 1 1>
        ]}
       GreenCurve {keys [
         <at 0 0.3005>
         <at 0.125 0>
         <at 0.375 0.0118>
         <at 0.625 1>
         {arriveTangent=0.462, interpMode=RCIM_Cubic, leaveTangent=0.462, time=0.875, value=1}
         <at 1 0.3005>
        ]}
       RedCurve {keys [
         <at 0 0>
         <at 0.125 0.5569>
         <at 0.375 1>
         <at 0.625 0.3333>
         {arriveTangent=-1.6, interpMode=RCIM_Cubic, leaveTangent=-1.6, time=0.875, value=0}
         <at 1 0>
        ]}
      }})) {isSetParameters=true, module=SetVariables_#4}
   &ScaleSpriteSize(
    Scale Sprite Size Mode=::NewEnumerator0
    Uniform Scale Factor=&Waveform(
     Global Phase=&Add_Float(A=$Emitter.Age, B=&RandomRangeFloat(Minimum=0, Maximum=5))
     "Waveform [1]"=::NewEnumerator0
     "[1] Frequency"=1
     "[1] Amplitude Min/Max"=<0.85 1>
    )
   ) {enabled=false}
   &SpriteRotationRate(Rotation Rate=&RandomRangeFloat(Minimum=10, Maximum=20))
   &SkeletalMeshLocation(
    Bone / Socket Index=&SelectIntFromArray(Direct Array Index=$Particles.PartitionParticles.Partition, Int Selection Array=DataInterface() {propertyValues {IntData [0,1,2,3,4,5,6]}})
    Bone Sampling Mode=::NewEnumerator4
    MaskSpawnGroup=false
    Orientation Sampling=::NewEnumerator1
    Sampled Position Offset Coordinate Space:ENiagaraSkelMeshTransforms=::NewEnumerator1
    Skeletal Mesh=$System.Skeletal Mesh
   ) {module=SkeletalMeshLocation003}
   &ScaleColor(
    Scale Alpha=0.4
    Scale RGB=&MakeVectorFromLinearColorRGB(Color=&ColorFromCurve(ColorCurve=DataInterface() {propertyValues {
        AlphaCurve {keys [<at 0 1 RCIM_Cubic>,<at 1 1 RCIM_Cubic>]}
        BlueCurve {keys [
          <at 0.0411 1>
          <at 0.3308 1>
          <at 0.5699 0>
          <at 0.7921 0>
          {arriveTangent=4.8091, interpMode=RCIM_Cubic, leaveTangent=4.8091, time=0.8998, value=0.7382}
          <at 1 1>
         ]}
        GreenCurve {keys [
          <at 0.0411 0.3005>
          <at 0.3308 0>
          <at 0.5699 0.0116>
          <at 0.7921 1>
          {arriveTangent=-3.3638, interpMode=RCIM_Cubic, leaveTangent=-3.3638, time=0.8998, value=1}
          <at 1 0.3005>
         ]}
        RedCurve {keys [
          <at 0.0411 0>
          <at 0.3308 0.5569>
          <at 0.5699 1>
          <at 0.7921 0.3333>
          {arriveTangent=-1.603, interpMode=RCIM_Cubic, leaveTangent=-1.603, time=0.8998, value=0}
          <at 1 0>
         ]}
       }}))
    ScaleA=true
    ScaleRGB=true
   )
   &DynamicMaterialParameters(
    Index 0 Param 1=&SelectFloatFromArray(Direct Array Index=$Particles.PartitionParticles.Partition, Float Selection Array=DataInterface() {propertyValues {FloatData [5,5,5,5,3,3]}})
    Index 0 Param 3=1
    Module.Param0WriteEnabled=true
    Module.Param1WriteEnabled=true
    Module.Param2WriteEnabled=true
    Module.Param3WriteEnabled=true
   )
  ]
  rendererClasses [&NiagaraSpriteRendererProperties]
  renderers [
   {rendererClass=&NiagaraSpriteRendererProperties}
   {rendererClass=&NiagaraSpriteRendererProperties, rendererIndex=1}
  ]
 }
 Rays {
  bEnabled=false
  emitterUpdateScript [
   &EmitterState(Inactive Response=::NewEnumerator1, Life Cycle Mode=::NewEnumerator1, Loop Duration=3.2, UseLoopDelay=false)
   &SpawnRate(SpawnRate=1.5, Use Spawn Probability=false)
  ]
  particleSpawnScript [
   &PartitionParticles(Number Of Partitions=4)
   &InitializeParticle(
    Color Mode=::NewEnumerator0
    Lifetime=6
    Sprite Rotation Angle=90
    Sprite Rotation Mode=::NewEnumerator0
    Sprite Size=&Multiply_Vector2D(
     A=<20 60>
     B=&MakeCustomVector2DFromBool(Bool=$User.IsLargeBody, True Vector2D=<1.2 1.2>, False Vector2D=<1 1>)
    )
    Sprite Size Mode=::NewEnumerator4
    UsePositionOffset=false
    Write Material Random=true
   )
  ]
  particleUpdateScript [
   &ParticleState()
   &CameraOffset()
   &SpriteRotationRate(Rotation Rate=50)
   &SkeletalMeshLocation(
    Bone / Socket Index=7
    Bone Sampling Mode=::NewEnumerator4
    MaskSpawnGroup=false
    Orientation Sampling=::NewEnumerator1
    Skeletal Mesh=$System.Skeletal Mesh
   )
   &SolveForcesAndVelocity(Clamp Velocity=false, Limit Acceleration=false)
   &ScaleColor(
    Scale Alpha=&FloatFromCurve(
     FloatCurve=DataInterface() {propertyValues {Curve {keys [
         {interpMode=RCIM_Cubic, tangentMode=RCTM_User, time=0, value=0}
         {arriveTangent=0.1935, interpMode=RCIM_Cubic, leaveTangent=0.1935, tangentMode=RCTM_User, time=0.2, value=1}
         {interpMode=RCIM_Cubic, tangentMode=RCTM_User, time=0.5, value=1}
         {tangentMode=RCTM_User, time=0.6, value=0}
        ]}}}
     Scale Curve=&FNTimeOfDayScalar(
      Night Scale=&RandomRangeFloat(Minimum=0.3, Maximum=0.4)
      Morning Scale=&RandomRangeFloat(Minimum=0.4, Maximum=0.5)
      Day Scale=&RandomRangeFloat(Minimum=0.4, Maximum=0.5)
      Evening Scale=&RandomRangeFloat(Minimum=0.4, Maximum=0.5)
     )
    )
    Scale RGB=&MakeVectorFromLinearColorRGB(Color=&SelectLinearColorFromArray(
      Array Sampling Mode=::NewEnumerator1
      Direct Array Index=$Particles.PartitionParticles.Partition
      Color Selection Array=DataInterface() {propertyValues {ColorData [<rgba 0 0.302 1 1>, <rgba 0.5569 0 1 1>, <rgba 1 0.0118 0 1>, <rgba 0.3333 1 0 1>]}}
      Random Seed=?
     ))
    ScaleA=true
    ScaleRGB=true
   )
  ]
  rendererClasses [&NiagaraSpriteRendererProperties]
  renderers [{rendererClass=&NiagaraSpriteRendererProperties}]
 }
 Rays_Burst {
  emitterUpdateScript [
   &EmitterState(
    Inactive Response=::NewEnumerator1
    Life Cycle Mode=::NewEnumerator1
    Loop Behavior=::NewEnumerator1
    Loop Duration Mode=::NewEnumerator1
    UseLoopDelay=false
   )
   &SpawnBurst_Instantaneous(Spawn Count=6, Use Spawn Probability=false, UseLoopCountLimit=false)
  ]
  particleSpawnScript [
   &PartitionParticles(Number Of Partitions=6)
   &InitializeParticle(
    Color=<rgba 3 3 3 1>
    Lifetime=&Lerp_Float(A=1, B=6, Alpha=&ReturnNormalizedExecIndex())
    Sprite Rotation Angle=&Lerp_Float(A=270, B=90, Alpha=&ReturnNormalizedExecIndex())
    Sprite Rotation Mode=::NewEnumerator0
    Sprite Size=&Multiply_Vector2D(
     A=&RandomRangeVector2D(Minimum=<13 50>, Maximum=<15 55>)
     B=&MakeCustomVector2DFromBool(Bool=$User.IsLargeBody, True Vector2D=<1.2 1.2>, False Vector2D=<1 1>)
    )
    Sprite Size Mode=::NewEnumerator4
    UsePositionOffset=false
    Write Material Random=true
   )
   /SetVariables(
    Particles.ThinRayRotation:NiagaraFloat=&Lerp_Float(A=253, B=73, Alpha=&ReturnNormalizedExecIndex())
    Particles.ThinRaySize:Vector2f=&RandomRangeVector2D(Minimum=<1 40>, Maximum=<3 55>)
   ) {isSetParameters=true}
   /SetVariables(
    Particles.GlowColor=<rgba 2.5 2.5 2.5 0.4>
    Particles.GlowRotation:NiagaraFloat=&Lerp_Float(A=250, B=78, Alpha=&ReturnNormalizedExecIndex())
    Particles.GlowSize=<35 45>
   ) {isSetParameters=true}
   &CameraOffset(Camera Offset Amount=-50) {module=CameraOffset001}
  ]
  particleUpdateScript [
   &ParticleState()
   &SpriteRotationRate(Rotation Rate=50)
   &ScaleSpriteSize(
    Scale Factor=&MakeVector2D(
     X=&FloatFromCurve(FloatCurve=DataInterface() {propertyValues {Curve {keys [
          {interpMode=RCIM_Cubic, tangentMode=RCTM_User, time=0, value=1}
          {interpMode=RCIM_Cubic, tangentMode=RCTM_User, time=0.5, value=1.33}
          {tangentMode=RCTM_User, time=1, value=0.9}
         ]}}})
     Y=&FloatFromCurve(FloatCurve=DataInterface() {propertyValues {Curve {keys [
          {interpMode=RCIM_Cubic, tangentMode=RCTM_User, time=0, value=1}
          {interpMode=RCIM_Cubic, tangentMode=RCTM_User, time=0.5, value=1.1}
          {tangentMode=RCTM_User, time=1, value=1}
         ]}}})
    )
    Scale Sprite Size Mode=::NewEnumerator2
   )
   &SkeletalMeshLocation(
    Bone / Socket Index=7
    Bone Sampling Mode=::NewEnumerator4
    MaskSpawnGroup=false
    Orientation Sampling=::NewEnumerator1
    Skeletal Mesh=$System.Skeletal Mesh
   )
   &SolveForcesAndVelocity(Clamp Velocity=false, Limit Acceleration=false)
   &ScaleColor(
    Scale Alpha=&FloatFromCurve(
     FloatCurve=DataInterface() {propertyValues {Curve {keys [
         {interpMode=RCIM_Cubic, tangentMode=RCTM_User, time=0, value=0}
         {arriveTangent=0.1935, interpMode=RCIM_Cubic, leaveTangent=0.1935, tangentMode=RCTM_User, time=0.05, value=1}
         {interpMode=RCIM_Cubic, tangentMode=RCTM_User, time=0.4, value=1}
         {tangentMode=RCTM_User, time=0.6, value=0}
        ]}}}
     Scale Curve=&FNTimeOfDayScalar(
      Night Scale=&RandomRangeFloat(Minimum=0.3, Maximum=0.4)
      Morning Scale=&RandomRangeFloat(Minimum=0.4, Maximum=0.5)
      Day Scale=&RandomRangeFloat(Minimum=0.8, Maximum=1)
      Evening Scale=&RandomRangeFloat(Minimum=0.4, Maximum=0.5)
     )
    )
    Scale RGB=&Multiply_Vector(
     A=&Multiply_Vector(
      A=&Multiply_Vector(
       A=&MakeVectorFromLinearColorRGB(Color=&SelectLinearColorFromArray(
         Array Sampling Mode=::NewEnumerator1
         Direct Array Index=$Particles.PartitionParticles.Partition
         Color Selection Array=DataInterface() {propertyValues {ColorData [<rgba 1 0.0116 0 1>, <rgba 1 0.2658 0 1>, <rgba 1 0.8377 0 1>, <rgba 0.1647 1 0 1>, <rgba 0 0.1429 1 1>, <rgba 0.5569 0 1 1>]}}
         Random Seed=?
        ))
       B=<1 1 1>
      )
      B=&VectorFromFloat(Value=&FNTimeOfDayScalar(
        Night Scale=1.15
        Morning Scale=1.15
        Day Scale=&MakeCustomFloatFromBool(Bool=$User.IsFrontEnd, True Float=0.6, False Float=1.75)
        Evening Scale=1.15
       ))
     )
     B=&VectorFromFloat(Value=&FloatFromCurve(FloatCurve=DataInterface() {propertyValues {Curve {keys [
           <at 0 1 RCIM_Cubic>
           {arriveTangentWeight=0.0667, interpMode=RCIM_Cubic, leaveTangentWeight=0.2667, tangentMode=RCTM_User, time=0.5, value=2}
           <at 1 1 RCIM_Cubic>
          ]}}}))
    )
    ScaleA=true
    ScaleRGB=true
   ) {module=ScaleColor001}
   &self:ScratchModule(Rotation Rate=50, ThinRayRotation:NiagaraFloat=$Particles.ThinRayRotation) {module=ScratchModule}
   /SetVariables(Particles.ThinRaySize:Vector2f=&Multiply_Vector2D(
     A=$Particles.Initial.ThinRaySize
     B=&MakeVector2D(
      X=&FloatFromCurve(FloatCurve=DataInterface() {propertyValues {Curve {keys [
           {interpMode=RCIM_Cubic, tangentMode=RCTM_User, time=0, value=1}
           {interpMode=RCIM_Cubic, tangentMode=RCTM_User, time=0.1, value=1.33}
           {arriveTangent=-0.2, interpMode=RCIM_Cubic, leaveTangent=-0.2, time=0.3, value=0.95}
           {arriveTangent=-0.375, interpMode=RCIM_Cubic, leaveTangent=-0.375, time=0.5, value=1.25}
           <at 0.7 0.8 RCIM_Cubic>
           {arriveTangent=0.6667, interpMode=RCIM_Cubic, leaveTangent=0.6667, time=0.8, value=1.25}
           {tangentMode=RCTM_User, time=1, value=1}
          ]}}})
      Y=&FloatFromCurve(FloatCurve=DataInterface() {propertyValues {Curve {keys [
           {interpMode=RCIM_Cubic, tangentMode=RCTM_User, time=0, value=1}
           {interpMode=RCIM_Cubic, tangentMode=RCTM_User, time=0.1, value=0.77}
           {arriveTangent=-0.05, interpMode=RCIM_Cubic, leaveTangent=-0.05, time=0.3, value=1.15}
           {arriveTangent=0.125, interpMode=RCIM_Cubic, leaveTangent=0.125, time=0.5, value=0.75}
           <at 0.7 1.2 RCIM_Cubic>
           {arriveTangent=-0.6667, interpMode=RCIM_Cubic, leaveTangent=-0.6667, time=0.8, value=0.75}
           {tangentMode=RCTM_User, time=1, value=1}
          ]}}})
     )
    )) {isSetParameters=true}
   /SetVariables(Particles.GlowColor=&MultiplyLinearColor(
     Linear Color=&MakeLinearColorFromVectorAndFloat(
      "Vector (RGB)"=&Multiply_Vector(
       A=&Multiply_Vector(
        A=&Multiply_Vector(
         A=&MakeVectorFromLinearColorRGB(Color=&SelectLinearColorFromArray(
           Array Sampling Mode=::NewEnumerator1
           Direct Array Index=$Particles.PartitionParticles.Partition
           Color Selection Array=DataInterface() {propertyValues {ColorData [<rgba 1 0.0116 0 1>, <rgba 1 0.2658 0 1>, <rgba 1 0.8377 0 1>, <rgba 0.1647 1 0 1>, <rgba 0 0.1429 1 1>, <rgba 0.5569 0 1 1>]}}
           Random Seed=?
          ))
         B=<1 1 1>
        )
        B=&VectorFromFloat(Value=&FNTimeOfDayScalar(
          Night Scale=1.15
          Morning Scale=1.15
          Day Scale=&MakeCustomFloatFromBool(Bool=$User.IsFrontEnd, True Float=0.4, False Float=1.3)
          Evening Scale=1.15
         ))
       )
       B=&VectorFromFloat(Value=&FloatFromCurve(FloatCurve=DataInterface() {propertyValues {Curve {keys [
             <at 0 1 RCIM_Cubic>
             {arriveTangentWeight=0.0667, interpMode=RCIM_Cubic, leaveTangentWeight=0.2667, tangentMode=RCTM_User, time=0.5, value=2}
             <at 1 1 RCIM_Cubic>
            ]}}}))
      )
      "Float (Alpha)"=&FloatFromCurve(
       FloatCurve=DataInterface() {propertyValues {Curve {keys [
           {interpMode=RCIM_Cubic, tangentMode=RCTM_User, time=0, value=0}
           {arriveTangent=0.1935, interpMode=RCIM_Cubic, leaveTangent=0.1935, tangentMode=RCTM_User, time=0.15, value=1}
           {interpMode=RCIM_Cubic, tangentMode=RCTM_User, time=0.45, value=1}
           {tangentMode=RCTM_User, time=0.65, value=0}
          ]}}}
       Scale Curve=&FNTimeOfDayScalar(
        Night Scale=&RandomRangeFloat(Minimum=0.3, Maximum=0.4)
        Morning Scale=&RandomRangeFloat(Minimum=0.4, Maximum=0.5)
        Day Scale=&RandomRangeFloat(Minimum=0.8, Maximum=1)
        Evening Scale=&RandomRangeFloat(Minimum=0.4, Maximum=0.5)
       )
      )
     )
     Scale Factor=&MakeVector4_FromLinearColor(LinearColor=$Particles.Initial.GlowColor)
    )) {isSetParameters=true}
  ]
  rendererClasses [&NiagaraSpriteRendererProperties]
  renderers [
   {rendererClass=&NiagaraSpriteRendererProperties}
   {rendererClass=&NiagaraSpriteRendererProperties, rendererIndex=1}
   {rendererClass=&NiagaraSpriteRendererProperties, rendererIndex=2}
  ]
 }
 Rays_Persistent {
  emitterUpdateScript [
   &EmitterState(Delay First Loop Only=true, Inactive Response=::NewEnumerator1, Life Cycle Mode=::NewEnumerator1, Loop Duration=3.2, UseLoopDelay=true)
   &SpawnRate(SpawnRate=1.5, Use Spawn Probability=false)
  ]
  particleSpawnScript [
   &PartitionParticles(Number Of Partitions=6)
   &InitializeParticle(
    Color=<rgba 3 3 3 1>
    Lifetime=6
    Sprite Rotation Angle=90
    Sprite Rotation Mode=::NewEnumerator0
    Sprite Size=&Multiply_Vector2D(
     A=&Multiply_Vector2D(A=&RandomRangeVector2D(Minimum=<13 55>, Maximum=<15 60>), B=<1 1>)
     B=&MakeCustomVector2DFromBool(Bool=$User.IsLargeBody, True Vector2D=<1.2 1.2>, False Vector2D=<1 1>)
    )
    Sprite Size Mode=::NewEnumerator4
    UsePositionOffset=false
    Write Material Random=true
   )
   /SetVariables(
    Particles.ThinRayRotation=72.5
    Particles.ThinRaySize:Vector2f=&RandomRangeVector2D(Minimum=<1 40>, Maximum=<3 55>)
   ) {isSetParameters=true}
   /SetVariables(Particles.GlowColor=<rgba 2.5 2.5 2.5 0.4>, Particles.GlowRotation=75, Particles.GlowSize=<40 45>) {isSetParameters=true}
   &CameraOffset(Camera Offset Amount=-50) {module=CameraOffset001}
  ]
  particleUpdateScript [
   &ParticleState()
   &ScaleSpriteSize(
    Scale Factor=&MakeVector2D(
     X=&FloatFromCurve(FloatCurve=DataInterface() {propertyValues {Curve {keys [
          {interpMode=RCIM_Cubic, tangentMode=RCTM_User, time=0, value=1}
          {interpMode=RCIM_Cubic, tangentMode=RCTM_User, time=0.33, value=1.1}
          <at 0.66 1 RCIM_Cubic>
         ]}}})
     Y=&FloatFromCurve(FloatCurve=DataInterface() {propertyValues {Curve {keys [
          {interpMode=RCIM_Cubic, tangentMode=RCTM_User, time=0, value=1}
          {interpMode=RCIM_Cubic, tangentMode=RCTM_User, time=0.25, value=1.1}
          {tangentMode=RCTM_User, time=0.5, value=1}
         ]}}})
    )
    Scale Sprite Size Mode=::NewEnumerator2
   )
   &SpriteRotationRate(Rotation Rate=50)
   &self:ScratchModule(Rotation Rate=50, ThinRayRotation:NiagaraFloat=$Particles.ThinRayRotation) {module=ScratchModule}
   &SkeletalMeshLocation(
    Bone / Socket Index=7
    Bone Sampling Mode=::NewEnumerator4
    MaskSpawnGroup=false
    Orientation Sampling=::NewEnumerator1
    Skeletal Mesh=$System.Skeletal Mesh
   )
   &SolveForcesAndVelocity(Clamp Velocity=false, Limit Acceleration=false)
   &ScaleColor(
    Scale Alpha=&FloatFromCurve(
     FloatCurve=DataInterface() {propertyValues {Curve {keys [
         {interpMode=RCIM_Cubic, tangentMode=RCTM_User, time=0, value=0}
         {arriveTangent=0.1935, interpMode=RCIM_Cubic, leaveTangent=0.1935, tangentMode=RCTM_User, time=0.2, value=1}
         {interpMode=RCIM_Cubic, tangentMode=RCTM_User, time=0.4, value=1}
         {tangentMode=RCTM_User, time=0.6, value=0}
        ]}}}
     Scale Curve=&FNTimeOfDayScalar(
      Night Scale=&RandomRangeFloat(Minimum=0.3, Maximum=0.4)
      Morning Scale=&RandomRangeFloat(Minimum=0.4, Maximum=0.5)
      Day Scale=&RandomRangeFloat(Minimum=0.8, Maximum=1)
      Evening Scale=&RandomRangeFloat(Minimum=0.4, Maximum=0.5)
     )
    )
    Scale RGB=&Multiply_Vector(
     A=&Multiply_Vector(
      A=&Multiply_Vector(
       A=&MakeVectorFromLinearColorRGB(Color=&SelectLinearColorFromArray(
         Array Sampling Mode=::NewEnumerator1
         Direct Array Index=$Particles.PartitionParticles.Partition
         Color Selection Array=DataInterface() {propertyValues {ColorData [<rgba 1 0.0116 0 1>, <rgba 1 0.2658 0 1>, <rgba 1 0.8377 0 1>, <rgba 0.1647 1 0 1>, <rgba 0 0.1429 1 1>, <rgba 0.5569 0 1 1>]}}
         Random Seed=?
        ))
       B=<1 1 1>
      )
      B=&VectorFromFloat(Value=&FNTimeOfDayScalar(
        Night Scale=1.15
        Morning Scale=1.15
        Day Scale=&MakeCustomFloatFromBool(Bool=$User.IsFrontEnd, True Float=0.6, False Float=1.75)
        Evening Scale=1.15
       ))
     )
     B=&VectorFromFloat(Value=&FloatFromCurve(FloatCurve=DataInterface() {propertyValues {Curve {keys [
           <at 0 1 RCIM_Cubic>
           {arriveTangentWeight=0.0667, interpMode=RCIM_Cubic, leaveTangentWeight=0.2667, tangentMode=RCTM_User, time=0.5, value=2}
           <at 1 1 RCIM_Cubic>
          ]}}}))
    )
    ScaleA=true
    ScaleRGB=true
   ) {module=ScaleColor001}
   /SetVariables(Particles.ThinRaySize:Vector2f=&Multiply_Vector2D(
     A=$Particles.Initial.ThinRaySize
     B=&MakeVector2D(
      X=&FloatFromCurve(FloatCurve=DataInterface() {propertyValues {Curve {keys [
           {interpMode=RCIM_Cubic, tangentMode=RCTM_User, time=0, value=1}
           {interpMode=RCIM_Cubic, tangentMode=RCTM_User, time=0.1, value=1.33}
           {arriveTangent=-0.2, interpMode=RCIM_Cubic, leaveTangent=-0.2, time=0.3, value=0.95}
           {arriveTangent=-0.375, interpMode=RCIM_Cubic, leaveTangent=-0.375, time=0.5, value=1.25}
           <at 0.7 0.8 RCIM_Cubic>
           {arriveTangent=0.6667, interpMode=RCIM_Cubic, leaveTangent=0.6667, time=0.8, value=1.25}
           {tangentMode=RCTM_User, time=1, value=1}
          ]}}})
      Y=&FloatFromCurve(FloatCurve=DataInterface() {propertyValues {Curve {keys [
           {interpMode=RCIM_Cubic, tangentMode=RCTM_User, time=0, value=1}
           {interpMode=RCIM_Cubic, tangentMode=RCTM_User, time=0.1, value=0.77}
           {arriveTangent=-0.05, interpMode=RCIM_Cubic, leaveTangent=-0.05, time=0.3, value=1.15}
           {arriveTangent=0.125, interpMode=RCIM_Cubic, leaveTangent=0.125, time=0.5, value=0.75}
           <at 0.7 1.2 RCIM_Cubic>
           {arriveTangent=-0.6667, interpMode=RCIM_Cubic, leaveTangent=-0.6667, time=0.8, value=0.75}
           {tangentMode=RCTM_User, time=1, value=1}
          ]}}})
     )
    )) {isSetParameters=true}
   /SetVariables(Particles.GlowColor=&MultiplyLinearColor(
     Linear Color=&MakeLinearColorFromVectorAndFloat(
      "Vector (RGB)"=&Multiply_Vector(
       A=&Multiply_Vector(
        A=&Multiply_Vector(
         A=&MakeVectorFromLinearColorRGB(Color=&SelectLinearColorFromArray(
           Array Sampling Mode=::NewEnumerator1
           Direct Array Index=$Particles.PartitionParticles.Partition
           Color Selection Array=DataInterface() {propertyValues {ColorData [<rgba 1 0.0116 0 1>, <rgba 1 0.2658 0 1>, <rgba 1 0.8377 0 1>, <rgba 0.1647 1 0 1>, <rgba 0 0.1429 1 1>, <rgba 0.5569 0 1 1>]}}
           Random Seed=?
          ))
         B=<1 1 1>
        )
        B=&VectorFromFloat(Value=&FNTimeOfDayScalar(
          Night Scale=1.15
          Morning Scale=1.15
          Day Scale=&MakeCustomFloatFromBool(Bool=$User.IsFrontEnd, True Float=0.4, False Float=1.3)
          Evening Scale=1.15
         ))
       )
       B=&VectorFromFloat(Value=&FloatFromCurve(FloatCurve=DataInterface() {propertyValues {Curve {keys [
             <at 0 1 RCIM_Cubic>
             {arriveTangentWeight=0.0667, interpMode=RCIM_Cubic, leaveTangentWeight=0.2667, tangentMode=RCTM_User, time=0.5, value=2}
             <at 1 1 RCIM_Cubic>
            ]}}}))
      )
      "Float (Alpha)"=&FloatFromCurve(
       FloatCurve=DataInterface() {propertyValues {Curve {keys [
           {interpMode=RCIM_Cubic, tangentMode=RCTM_User, time=0, value=0}
           {arriveTangent=0.1935, interpMode=RCIM_Cubic, leaveTangent=0.1935, tangentMode=RCTM_User, time=0.15, value=1}
           {interpMode=RCIM_Cubic, tangentMode=RCTM_User, time=0.45, value=1}
           {tangentMode=RCTM_User, time=0.65, value=0}
          ]}}}
       Scale Curve=&FNTimeOfDayScalar(
        Night Scale=&RandomRangeFloat(Minimum=0.3, Maximum=0.4)
        Morning Scale=&RandomRangeFloat(Minimum=0.4, Maximum=0.5)
        Day Scale=&RandomRangeFloat(Minimum=0.8, Maximum=1)
        Evening Scale=&RandomRangeFloat(Minimum=0.4, Maximum=0.5)
       )
      )
     )
     Scale Factor=&MakeVector4_FromLinearColor(LinearColor=$Particles.Initial.GlowColor)
    )) {isSetParameters=true}
  ]
  rendererClasses [&NiagaraSpriteRendererProperties]
  renderers [
   {rendererClass=&NiagaraSpriteRendererProperties}
   {rendererClass=&NiagaraSpriteRendererProperties, rendererIndex=1}
   {rendererClass=&NiagaraSpriteRendererProperties, rendererIndex=2}
  ]
 }
 Sparkles001 {
  bEnabled=false
  emitterUpdateScript [
   &EmitterState(Delay First Loop Only=true, Life Cycle Mode=::NewEnumerator1, Loop Delay=0.1, Loop Duration=1, UseLoopDelay=true)
   &SpawnRate(SpawnRate=15, Use Spawn Probability=true)
  ]
  particleSpawnScript [
   &PartitionParticles(Number Of Partitions=7)
   &InitializeParticle(
    Lifetime Max=1.7
    Lifetime Mode=::NewEnumerator1
    Sprite Size Mode=::NewEnumerator1
    Uniform Sprite Size Max=35
    Uniform Sprite Size Min=20
    UsePositionOffset=false
    Write Material Random=true
   )
   &SkeletalMeshLocation(
    Bone / Socket Index=&SelectIntFromArray(Direct Array Index=$Particles.PartitionParticles.Partition, Int Selection Array=DataInterface() {propertyValues {IntData [0,1,2,3,4,5,6]}})
    Bone / Socket Sampling Mode=::NewEnumerator5
    Interpolate Mesh Sampling=false
    MaskSpawnGroup=false
    Mesh Sampling Type=::NewEnumerator6
    Orientation Sampling=::NewEnumerator1
    Position Sampling=::NewEnumerator1
    Sampled Position Offset=<0 0 5>
    Sampled Position Offset Coordinate Space:ENiagaraSkelMeshTransforms=::NewEnumerator1
    Skeletal Mesh=$System.Skeletal Mesh
    Source Movement Scale:Vector3f=<0 0 0>
   )
   &ShapeLocation(MaskSpawnGroup=false, Override Seed=false, Shape Origin=$Particles.Position, Sphere Radius=4, UseOwnerScale=true) {module=ShapeLocation001}
   &AddVelocity(Velocity=&RandomRangeVector(Minimum=<5 5 10>, Maximum=<-5 -5 60>))
   &AddVelocity(
    Module.Use Falloff Distance=false
    Override Seed=false
    Velocity Mode=::NewEnumerator1
    Velocity Origin=$Particles.Position
    Velocity Speed=&RandomRangeFloat(Minimum=-5, Maximum=5)
   ) {module=AddVelocity001}
  ]
  particleUpdateScript [
   &Drag(
    Drag=&FloatFromCurve(FloatCurve=DataInterface() {propertyValues {Curve {keys [
         {arriveTangent=1.5197, interpMode=RCIM_Cubic, leaveTangent=1.5197, tangentMode=RCTM_User, time=0, value=0}
         {arriveTangent=-0.2878, interpMode=RCIM_Cubic, tangentMode=RCTM_Break, time=1, value=1}
        ]}}})
    Use Linear Drag=true
    Use Rotational Drag=true
   )
   &SolveForcesAndVelocity(Clamp Velocity=false, Limit Acceleration=false)
   &ParticleState()
   &ScaleSpriteSize(
    Scale Sprite Size Mode=::NewEnumerator0
    Uniform Scale Factor=&FloatFromCurve(
     FloatCurve=DataInterface() {propertyValues {Curve {keys [
         {arriveTangent=21.1265, interpMode=RCIM_Cubic, leaveTangent=21.1265, tangentMode=RCTM_User, time=0, value=0}
         {arriveTangent=-2.5249, interpMode=RCIM_Cubic, leaveTangent=-2.5249, tangentMode=RCTM_User, time=0.1408, value=0.9847}
         {arriveTangent=-0.0405, interpMode=RCIM_Cubic, leaveTangent=-0.0405, tangentMode=RCTM_User, time=1, value=0}
        ]}}}
     Scale Curve=&RandomRangeFloat(Minimum=1, Maximum=2)
    )
   )
   /SetVariables(Particles.Glow_SpriteSize:Vector2f=&Multiply_Vector2D(A=$Particles.SpriteSize, B=<1.1 1.1>)) {isSetParameters=true, module=SetVariables_#5}
   &CameraOffset(Camera Offset Amount=-20)
   &ScaleColor(
    Scale Alpha=&FNTimeOfDayScalar(Night Scale=0.8, Morning Scale=1.1, Day Scale=1.2, Evening Scale=1)
    Scale RGB=&MakeVectorFromLinearColorRGB(Color=&SelectLinearColorFromArray(
      Array Sampling Mode=::NewEnumerator0
      Direct Array Index=?
      Color Selection Array=DataInterface() {propertyValues {ColorData [<rgba 0 0.302 1 1>, <rgba 1 0.8958 0 1>, <rgba 0.3333 1 0 1>, <rgba 1 1 1 1>]}}
      Random Seed=0
     ))
    ScaleA=true
    ScaleRGB=true
   )
   /SetVariables(Particles.Color2=&MultiplyLinearColor(Linear Color=$Particles.Color, Scale Factor=<0.5 1 1 0.4>)) {isSetParameters=true, module=SetVariables_#6}
  ]
  rendererClasses [&NiagaraSpriteRendererProperties]
  renderers [
   {rendererClass=&NiagaraSpriteRendererProperties}
   {rendererClass=&NiagaraSpriteRendererProperties, rendererIndex=1}
  ]
 }
 Sparkles002 {
  emitterUpdateScript [
   &EmitterState(Delay First Loop Only=true, Life Cycle Mode=::NewEnumerator1, Loop Delay=0.1, Loop Duration=1, UseLoopDelay=true)
   &SpawnRate(SpawnRate=25, Use Spawn Probability=true)
  ]
  particleSpawnScript [
   &PartitionParticles(Number Of Partitions=1)
   &InitializeParticle(
    Lifetime Max=1.7
    Lifetime Mode=::NewEnumerator1
    Sprite Rotation Mode=::NewEnumerator2
    Sprite Size Mode=::NewEnumerator1
    Uniform Sprite Size Max=35
    Uniform Sprite Size Min=17.5
    UsePositionOffset=false
    Write Material Random=true
   )
   &SkeletalMeshLocation(
    Bone / Socket Index=&SelectIntFromArray(Direct Array Index=$Particles.PartitionParticles.Partition, Int Selection Array=DataInterface() {propertyValues {IntData [7]}})
    Bone / Socket Sampling Mode=::NewEnumerator5
    Interpolate Mesh Sampling=false
    MaskSpawnGroup=false
    Mesh Sampling Type=::NewEnumerator6
    Orientation Sampling=::NewEnumerator1
    Position Sampling=::NewEnumerator1
    Sampled Position Offset=<0 0 5>
    Sampled Position Offset Coordinate Space:ENiagaraSkelMeshTransforms=::NewEnumerator1
    Skeletal Mesh=$System.Skeletal Mesh
    Source Movement Scale:Vector3f=<0 0 0>
   )
   &AddVelocity(
    Velocity=&RandomRangeVector(Minimum=<64 12.8 64>, Maximum=<-64 -64 -51.2>)
    Velocity Speed Scale=&MakeCustomFloatFromBool(Bool=$User.IsLargeBody, True Float=1.1, False Float=1)
   )
  ]
  particleUpdateScript [
   &Drag(
    Drag=&FloatFromCurve(
     FloatCurve=DataInterface() {propertyValues {Curve {keys [
         {arriveTangent=2.8093, interpMode=RCIM_Cubic, leaveTangent=2.8093, tangentMode=RCTM_User, time=0, value=0}
         {arriveTangent=-0.009, interpMode=RCIM_Cubic, tangentMode=RCTM_Break, time=0.8, value=1}
        ]}}}
     Scale Curve=2
    )
    Use Linear Drag=true
    Use Rotational Drag=true
   )
   &CurlNoiseForce(
    Noise Strength=&FloatFromCurve(FloatCurve=DataInterface() {propertyValues {Curve {keys [
         {arriveTangent=99.3036, interpMode=RCIM_Cubic, leaveTangent=99.3036, tangentMode=RCTM_User, time=0, value=0}
         {arriveTangent=-0.4476, interpMode=RCIM_Cubic, leaveTangent=-0.4476, tangentMode=RCTM_User, time=1, value=50}
        ]}}})
    Pan Noise=true
    Pan Noise Field=<0.1 0.1 0.1>
   )
   &SolveForcesAndVelocity(Clamp Velocity=false, Limit Acceleration=false)
   &ParticleState()
   &ScaleSpriteSize(
    Scale Sprite Size Mode=::NewEnumerator0
    Uniform Scale Factor=&FloatFromCurve(
     FloatCurve=DataInterface() {propertyValues {Curve {keys [
         {arriveTangent=21.1265, interpMode=RCIM_Cubic, leaveTangent=21.1265, tangentMode=RCTM_User, time=0, value=0}
         {arriveTangent=-2.5249, interpMode=RCIM_Cubic, leaveTangent=-2.5249, tangentMode=RCTM_User, time=0.1408, value=0.9847}
         {arriveTangent=-0.0405, interpMode=RCIM_Cubic, leaveTangent=-0.0405, tangentMode=RCTM_User, time=1, value=0}
        ]}}}
     Scale Curve=&RandomRangeFloat(Minimum=1, Maximum=2)
    )
   )
   /SetVariables(Particles.Glow_SpriteSize:Vector2f=&Multiply_Vector2D(A=$Particles.SpriteSize, B=<1 1>)) {isSetParameters=true, module=SetVariables_#5}
   &CameraOffset(Camera Offset Amount=-20)
   &ScaleColor(
    Scale Alpha=&FNTimeOfDayScalar(Night Scale=0.8, Morning Scale=1.1, Day Scale=1.2, Evening Scale=1)
    Scale RGB=&Multiply_Vector(
     A=&MakeVectorFromLinearColorRGB(Color=&SelectLinearColorFromArray(
       Array Sampling Mode=::NewEnumerator0
       Direct Array Index=?
       Color Selection Array=DataInterface() {propertyValues {ColorData [<rgba 1 0.0116 0 1>, <rgba 1 0.2658 0 1>, <rgba 1 0.8377 0 1>, <rgba 0.1647 1 0 1>, <rgba 0 0.1429 1 1>, <rgba 0.5569 0 1 1>]}}
       Random Seed=0
      ))
     B=&VectorFromFloat(Value=&FNTimeOfDayScalar(Night Scale=1, Morning Scale=1, Day Scale=2, Evening Scale=1))
    )
    ScaleA=true
    ScaleRGB=true
   )
   /SetVariables(Particles.Color2=&MultiplyLinearColor(Linear Color=$Particles.Color, Scale Factor=<0.5 1 1 0.4>)) {isSetParameters=true, module=SetVariables_#6}
  ]
  rendererClasses [&NiagaraSpriteRendererProperties]
  renderers [
   {rendererClass=&NiagaraSpriteRendererProperties}
   {rendererClass=&NiagaraSpriteRendererProperties, rendererIndex=1}
  ]
 }
}
summary {
 systemName=NS_Aura_Rainbow
 emitters [
  {bEnabled=false, emitterName=Rays, rendererClasses [&NiagaraSpriteRendererProperties]}
  {bEnabled=false, emitterName=Aurora_Mesh001, rendererClasses [&NiagaraMeshRendererProperties]}
  {bEnabled=false, emitterName=Glow001, rendererClasses [&NiagaraSpriteRendererProperties]}
  {bEnabled=false, emitterName=Sparkles001, rendererClasses [&NiagaraSpriteRendererProperties]}
  {bEnabled=false, emitterName=Aurora_Mesh_Scaling, rendererClasses [&NiagaraMeshRendererProperties]}
  {emitterName=Sparkles002, rendererClasses [&NiagaraSpriteRendererProperties]}
  {emitterName=Glow002, rendererClasses [&NiagaraSpriteRendererProperties]}
  {emitterName=Rays_Burst, rendererClasses [&NiagaraSpriteRendererProperties]}
  {emitterName=Rays_Persistent, rendererClasses [&NiagaraSpriteRendererProperties]}
  {bEnabled=false, emitterName=FN_Debug, rendererClasses [&NiagaraSpriteRendererProperties]}
 ]
 userVariables [
  User.bIsFrontEnd {type=NiagaraBool}
  User.bIsFrontEndPreview {type=NiagaraBool}
  User.IsFrontEnd {type=NiagaraBool}
  User.IsLargeBody {type=NiagaraBool}
 ]
}
